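# Translucent lit, two-sided Niagara particle material for propeller motion blur on the Lars Male pickaxe, driven by a SpinAmount dynamic parameter.
# Repeating U-axis gradient bands, offset by a threshold lerped between min and max by SpinAmount, smoothstep-cut a panning blurred noise texture, multiplied by particle alpha.
# A stepped offset toward the particle position collapses the geometry when SpinAmount is near zero.
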Material /BRCosmetics/Weapons/FORT_Melee/Pickaxe_Lars_Male/FX/M_FX_Pickaxe_LarsMale_PropMoblur
graph {
 properties {
  bEnableResponsiveAA=true
  bUsedWithNiagaraMeshParticles=true
  bUsedWithNiagaraSprites=true
  blendMode=BLEND_Translucent
  pixelDepthOffsetMode=PDOM_Legacy
  twoSided=true
  usedShadingModels=Default Lit
 }
 nodes [
  Constant3Vector() {constant=<rgba 0.1009 0.1042 0.1009 1>}
  Constant() {r=0.8}
  Constant() {r=0.2}
  LinearInterpolate(
   A=Constant3Vector() {constant=<rgba 1 0 0 0>}
   B=Constant3Vector() {constant=<rgba 0 1 0 0>}
   Alpha=DynamicParameter() {paramNames [SpinAmount,Param2,Param3,Param4]}.SpinAmount
  )
  n1 = DynamicParameter() {paramNames [SpinAmount,Param2,Param3,Param4]}
  Multiply(A=Subtract(A=ParticlePositionWS(), B=WorldPosition().XYZ), B=Step(Y=@n1.SpinAmount) {constX=0.001})
  n2 = Multiply(A=ComponentMask(None=TextureCoordinate()) {r=true}, B=ScalarParameter() {defaultValue=2, parameterName=NumberOfMoblur}) {constB=2}
  n3 = Constant() {r=1}
  n4 = Add(
   A=OneMinus(None=Fmod(A=@n2, B=@n3))
   B=LinearInterpolate(A=ScalarParameter() {parameterName=Step_MinMoblur}, B=ScalarParameter() {parameterName=Step_MaxMoblur}, Alpha=@n1.SpinAmount)
  )
  n5 = ScalarParameter() {parameterName=SStep_W}
  n6 = Panner(Coordinate=TextureCoordinate() {uTiling=0.1, vTiling=0.35}) {speedX=0.6}
  Multiply(
   A=ParticleColor().A
   B=SmoothStep(
    Min=Subtract(A=@n4, B=@n5)
    Max=Add(A=@n4, B=@n5)
    Value=TextureSample(UVs=@n6) {samplerType=SAMPLERTYPE_Grayscale, texture=&LowResBlurredNoise}.G
   )
  )
  ScalarParameter() {parameterName=Shrink}
 ]
}
