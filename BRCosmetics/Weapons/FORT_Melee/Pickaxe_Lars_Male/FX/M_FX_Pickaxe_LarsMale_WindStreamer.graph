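# Translucent, unlit, two-sided material for wind streamer mesh particles on the Lars Male pickaxe effect, colored by particle color through emissive adjustments.
# Opacity-style chain multiplies particle alpha, soft UV mask, panning noise and view-angle fade.
# A sphere mask along V, positioned by the RevealArea dynamic parameter, reveals the streak; vertex-interpolated camera-proximity fade and time-of-day scalar attenuate it.
# SplineThicken and a UV-offset branch provide width and thickening math.
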Material /BRCosmetics/Weapons/FORT_Melee/Pickaxe_Lars_Male/FX/M_FX_Pickaxe_LarsMale_WindStreamer
graph {
 properties {
  bUsedWithNiagaraMeshParticles=true
  blendMode=BLEND_Translucent
  pixelDepthOffsetMode=PDOM_Legacy
  shadingModel=MSM_Unlit
  twoSided=true
  usedShadingModels=Unlit
 }
 nodes [
  Constant3Vector() {constant=<rgba 0 1 0 0>}
  n1 = ScalarParameter() {parameterName=Width}
  MaterialFunctionCall(WidthBase=@n1, WidthTip=@n1) {materialFunction=&SplineThicken}
  n2 = ParticleColor()
  MaterialFunctionCall(Emissive=@n2.RGB) {materialFunction=&MF_FX_EmissiveAdjustments}
  n3 = Multiply(A=Subtract(A=TextureCoordinate()) {constB=0.5}, B=Constant2Vector() {r=1})
  n4 = Constant()
  Multiply(A=Transform(None=AppendVector(A=@n3, B=@n4)), B=ScalarParameter() {parameterName=Thicken})
  n5 = SphereMask(A=ComponentMask(None=TextureCoordinate()) {g=true}, B=DynamicParameter() {paramNames [RevealArea,Param2,Param3,Param4]}.RevealArea) {attenuationRadius=0.35, hardnessPercent=0}
  n6 = Power(Base=SphereMask(A=TextureCoordinate(), B=Constant() {r=0.5}) {attenuationRadius=0.5, hardnessPercent=0})
  n7 = TextureSample(UVs=Panner(Coordinate=TextureCoordinate() {uTiling=0.75, vTiling=0.2}) {speedY=-0.25}) {samplerType=SAMPLERTYPE_LinearColor, texture=&T_NoisePack_LinearColor}
  n8 = Multiply(
   A=Multiply(A=@n2.A, B=Multiply(A=@n6, B=@n7.G))
   B=MaterialFunctionCall("FadeSensitivity (optional)"=ScalarParameter() {parameterName=fade}) {materialFunction=&FadeBasedOnViewAngle}.Result
  )
  n9 = OneMinus(None=SphereMask(A=CameraPositionWS(), B=WorldPosition().XYZ, Radius=Constant() {r=100}) {hardnessPercent=0})
  n10 = Constant() {r=0.35}
  n11 = Constant() {r=0.75}
  n12 = Constant() {r=0.65}
  n13 = Constant() {r=0.75}
  Multiply(
   A=Multiply(A=Multiply(A=@n5, B=@n8), B=VertexInterpolator(VS=@n9).PS)
   B=VertexInterpolator(VS=MaterialFunctionCall(Value_Morning=@n10, Value_Day=@n11, Value_Evening=@n12, Value_Night=@n13) {materialFunction=&Scalar_Time_of_Day}.Result).PS
  )
 ]
}
