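# Physics asset for the Lars male pickaxe mesh, defining bodies and joints for three wire bones, apparently for secondary dangling motion.
# A small capsule sits on the root bone, and each wire bone has its own offset capsule, with wireB slightly thicker.
# Each wire is constrained directly to the root rather than chained, with matching swing and twist limit angles.
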PhysicsAsset /BRCosmetics/Weapons/FORT_Melee/Pickaxe_Lars_Male/Meshes/Lars_Male_Axe_Physics
bodies {
 b_AR_Root [{shapeType=Capsule, length=2, radius=2, shapeName=b_AR_Root_capsule}]
 dyn_wireA [{shapeType=Capsule, length=6.25, radius=2, shapeName=dyn_wireA_capsule, center=<5 0 0>}]
 dyn_wireB [{shapeType=Capsule, length=6.25, radius=2.75, shapeName=dyn_wireB_capsule, center=<5 0 0>}]
 dyn_wireC [{shapeType=Capsule, length=6.25, radius=2, shapeName=dyn_wireC_capsule, center=<5 0 0>}]
}
constraints [
 {bone1Name=dyn_wireA, bone2Name=b_AR_Root, swing1LimitDegrees=45, swing2LimitDegrees=45, twistLimitDegrees=45}
 {bone1Name=dyn_wireC, bone2Name=b_AR_Root, swing1LimitDegrees=45, swing2LimitDegrees=45, twistLimitDegrees=45}
 {bone1Name=dyn_wireB, bone2Name=b_AR_Root, swing1LimitDegrees=45, swing2LimitDegrees=45, twistLimitDegrees=45}
]
