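# Animation Blueprint for the Lars male pickaxe, animating its propeller bones.
# The update event eases spin speeds up during melee swings and 2D movement, accumulates them into rotators, and enables a constant time-based spin in the front end.
# Additive bone transforms apply these spins, then a lightly weighted RigidBody node adds physics.
# The velocity-based RunRot layer has zero alpha, and the propellerBlur rotation logic is not wired to the update event.
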AnimBlueprint /BRCosmetics/Weapons/FORT_Melee/Pickaxe_Lars_Male/Meshes/Lars_Male_Axe_AnimBP
target_skeleton=&/Lars_Male_Axe_Skeleton
graphs [
 AnimGraph {nodes [
   AnimGraphNode_SubInput_0 = Input Pose()
   n1 = "Transform (Modify) Bone - Bone: propeller"(
    ComponentPose=Local To Component(LocalPose=@AnimGraphNode_SubInput_0)
    Alpha=1
    Translation="0,0,0"
    Rotation=Get Spinning_Rotator_00() -> (Spinning_Rotator_00:Rotator)
    Scale="1,1,1"
   ) {
    rotationMode=BMM_Additive
    rotationSpace=BCS_BoneSpace
    alphaBoolBlend {blendOption=Linear}
    alphaScaleBiasClamp {interpSpeedDecreasing=0, interpSpeedIncreasing=50}
    boneToModify {boneName=propeller}
   }
   n2 = "Transform (Modify) Bone - Bone: propellerBlur"(ComponentPose=@n1, Alpha=1, Translation="0,0,0", Rotation=Get Blur() -> (Blur:Rotator), Scale="1,1,1") {rotationMode=BMM_Additive, rotationSpace=BCS_ParentBoneSpace, alphaBoolBlend {blendOption=Linear}, boneToModify {boneName=propellerBlur}}
   n3 = "Transform (Modify) Bone - Bone: propeller"(ComponentPose=@n2, Alpha=0, Translation="0,0,0", Rotation=Get RunRot() -> (RunRot:Rotator), Scale="1,1,1") {rotationMode=BMM_Additive, rotationSpace=BCS_ParentBoneSpace, alphaBoolBlend {blendOption=Linear}, boneToModify {boneName=propeller}}
   n4 = "Transform (Modify) Bone - Bone: propeller"(ComponentPose=@n3, Alpha=1, Translation="0,0,0", Rotation=Get isMovingRot() -> (isMovingRot:Rotator), Scale="1,1,1") {rotationMode=BMM_Additive, rotationSpace=BCS_ParentBoneSpace, alphaBoolBlend {blendOption=Linear}, boneToModify {boneName=propeller}}
   n5 = "Transform (Modify) Bone - Bone: propeller"(
    ComponentPose=@n4
    Alpha=Get frontEndAlpha() -> (frontEndAlpha:double)
    Translation="0,0,0"
    Rotation=Get ContinuousSpin() -> (ContinuousSpin:Rotator)
    Scale="1,1,1"
   ) {rotationMode=BMM_Additive, rotationSpace=BCS_ParentBoneSpace, alphaBoolBlend {blendOption=Linear}, boneToModify {boneName=propeller}}
   n6 = RigidBody(ComponentPose=@n5, Alpha=0.3) {
    bClampLinearTranslationLimitToRefPose=true
    bOverrideWorldGravity=true
    overridePhysicsAsset=&/Lars_Male_Axe_Physics
    simulationSpace=BaseBoneSpace
    alphaBoolBlend {blendOption=Linear}
    baseBoneRef {boneName=b_AR_Root}
    componentAppliedLinearAccClamp=<3000 3000 3000>
    componentLinearAccScale=<0.4 0.3 0.3>
    componentLinearVelScale=<0.4 0.3 0.3>
    overrideWorldGravity {z=-980}
   }
   Output Pose(Result=Component To Local(ComponentPose=@n6)) {layerGroup=None}
  ]}
 EventGraph {nodes [
   n1 = Event BlueprintUpdateAnimation()
   > n2 = Sequence!(execute) -> (then_0, then_1, then_2:Exec, then_3:Exec, then_4:Exec, then_5:Exec, then_6:Exec, then_7:Exec)
   > n3 = Branch(Condition=Get bIsPlayingMeleeAnim())
   > Set crackSpinSpeed(crackSpinSpeed:double=FInterpTo(Current=Get crackSpinSpeed() -> (crackSpinSpeed:double), Target=50, DeltaTime=@n1.DeltaTimeX, InterpSpeed=10)) -> (Output_Get:double)
   Set crackSpinSpeed(execute=@n3.else, crackSpinSpeed=FInterpTo(Current=Get crackSpinSpeed(), Target=0, DeltaTime=@n1.DeltaTimeX, InterpSpeed=1))
   n4 = Set crackSpin(
    execute=@n2.then_1
    crackSpin:double=float + float(A=Get crackSpin() -> (crackSpin:double), B=Get crackSpinSpeed()) {is=title K2Node_CommutativeAssociativeBinaryOperator}
   ) -> (Output_Get:double)
   > Set Spinning_Rotator_00(Spinning_Rotator_00:Rotator=MakeRotator(Roll=NormalizeAxis!(Angle:float=@n4) -> (ReturnValue:float))) -> (Output_Get:Rotator)
   n5 = Branch(execute=@n2.then_3, Condition=Get bIsMoving2D())
   > Set runSpeed(runSpeed:double=FInterpTo(Current=Get runSpeed() -> (runSpeed:double), Target=40, DeltaTime=@n1.DeltaTimeX, InterpSpeed=1)) -> (Output_Get:double)
   Set runSpeed(execute=@n5.else, runSpeed=FInterpTo(Current=Get runSpeed(), Target=0, DeltaTime=@n1.DeltaTimeX, InterpSpeed=1))
   Set Pawn Speed 2D(execute=@n2.then_4, Pawn Speed 2D:double=Get PawnSpeed2D()) -> (Output_Get:double)
   n6 = Set isMoving(execute=@n2.then_5, isMoving:double=float + float(A=Get runSpeed(), B=Get isMoving() -> (isMoving:double))) -> (Output_Get:double)
   > Set isMovingRot(isMovingRot:Rotator=MakeRotator(Pitch=NormalizeAxis(Angle=@n6))) -> (Output_Get:Rotator)
   n7 = Branch(execute=@n2.then_6, Condition=Get bIsInFrontEnd())
   > Set frontEndAlpha(frontEndAlpha:double=1) -> (Output_Get:double)
   Set frontEndAlpha(execute=@n7.else, frontEndAlpha=0)
   n8 = GetRealTimeSeconds()
   Set ContinuousSpin(execute=@n2.then_7, ContinuousSpin:Rotator=MakeRotator(Pitch=float * float(A=@n8, B=250))) -> (Output_Get:Rotator)
   n9 = GetPlayerCharacter!(PlayerIndex:int=0) -> (ReturnValue:Character Object Reference)
   Is Valid(InputObject=@n9)
   n10 = GetVelocity!(self:Actor Object Reference=@n9) -> (ReturnValue:Vector)
   n11 = BreakVector(InVec=@n10)
   n12 = MakeVector(X=@n11, Y=0, Z=@n11.Z)
   > n13 = Set PawSpeedAlpha(PawSpeedAlpha:double=float * float(A=Vector Length(A=@n12), B=0.05)) -> (Output_Get:double)
   > n14 = Set RunRotCombined(RunRotCombined:double=float + float(A=@n13, B=Get RunRotCombined() -> (RunRotCombined:double))) -> (Output_Get:double)
   > Set RunRot(RunRot:Rotator=MakeRotator(Pitch=NormalizeAxis(Angle=@n14))) -> (Output_Get:Rotator)
   n15 = Set BlurFloat(BlurFloat:double=MapRangeClamped(Value=Get crackSpinSpeed(), InRangeA=35, InRangeB=50, OutRangeB=-15)) -> (Output_Get:double)
   > Set Blur(Blur:Rotator=MakeRotator(Roll=@n15)) -> (Output_Get:Rotator)
   > n16 = Branch(Condition=Get bIsPlayingMeleeAnim())
   > n17 = Set BlurFloat(BlurFloat=FInterpTo(Current=Get BlurFloat() -> (BlurFloat:double), Target=-30, DeltaTime=@n1.DeltaTimeX, InterpSpeed=1))
   > Set Blur(Blur=MakeRotator(Roll=@n17))
   n18 = Set BlurFloat(execute=@n16.else, BlurFloat=FInterpTo(Current=Get BlurFloat(), Target=0, DeltaTime=@n1.DeltaTimeX, InterpSpeed=10))
   > Set Blur(Blur=MakeRotator(Roll=@n18))
   SelectFloat(A=0, B=1, bPickA="Not Equal (Float)"!(A:double=@n11.Z, B:double=0) -> (ReturnValue:bool) {is=title K2Node_CallFunction})
   n19 = Get Actor Rotation!(self:Actor Object Reference=@n9) -> (ReturnValue:Rotator) {is=title K2Node_CallFunction}
   n20 = BreakVector(InVec=vector * vector(A:Vector=Get Rotation X Vector!(InRot:Rotator=@n19) -> (ReturnValue:Vector) {is=title K2Node_CallFunction}, B:Vector=@n10) -> (ReturnValue:Vector) {is=title K2Node_CommutativeAssociativeBinaryOperator})
   float + float(A=@n20, B=@n20.Y)
   float * float(A=MapRangeClamped(Value=0, InRangeA=0, InRangeB=500), B=1)
   Reroute Node(InputPin:Wildcard) -> (OutputPin:Wildcard)
   Reroute Node()
  ]}
]
variables [
 Blur {type=Rotator}
 BlurFloat {type=number}
 ContinuousSpin {type=Rotator}
 crackSpin {type=number}
 crackSpinSpeed {type=number}
 frontEndAlpha {type=number}
 isMoving {type=number}
 isMovingRot {type=Rotator}
 isPickaxingAlpha {type=number}
 NewVar_0 {type=number}
 Pawn Speed 2D {type=number}
 PawSpeedAlpha {type=number}
 propRotate {type=Rotator}
 RunRot {type=Rotator}
 RunRotCombined {type=number}
 runSpeed {type=number}
 Spinning_Rotator_00 {type=Rotator}
]
